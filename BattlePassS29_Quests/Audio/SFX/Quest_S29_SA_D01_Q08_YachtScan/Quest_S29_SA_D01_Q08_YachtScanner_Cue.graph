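# Sound cue for the yacht scanner effect in a Season 29 Battle Pass quest.
# A random node with equal weights picks one of three scanner wave variations each time it plays.
# Playback uses a quest-specific attenuation asset, a slightly boosted volume and the Fort_Quests sound class.
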SoundCue /BattlePassS29_Quests/Audio/SFX/Quest_S29_SA_D01_Q08_YachtScan/Quest_S29_SA_D01_Q08_YachtScanner_Cue
root=Random_0
nodes [
 {id=Random_0, children [WavePlayer_0,WavePlayer_1,WavePlayer_2], properties {weights [1,1,1]}}
 {
  id=WavePlayer_0
  properties {soundWaveAssetPtr=&/Quest_S29_SA_D01_Q08_YachtScanner_03}
 }
 {
  id=WavePlayer_1
  properties {soundWaveAssetPtr=&/Quest_S29_SA_D01_Q08_YachtScanner_01}
 }
 {
  id=WavePlayer_2
  properties {soundWaveAssetPtr=&/Quest_S29_SA_D01_Q08_YachtScanner_02}
 }
]
properties {
 attenuationSettings=&/S29_SA_D01_Q08_Atten
 maxDistance=3050
 soundClassObject=&Fort_Quests
 volumeMultiplier=1.2
}
